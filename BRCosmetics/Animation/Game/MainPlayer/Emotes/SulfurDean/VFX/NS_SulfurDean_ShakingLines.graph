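# Niagara system for the SulfurDean emote that draws short-lived, near-black, jittering sprite speed lines around the character to suggest shaking.
# A burst of eleven particles is partitioned so each gets its own index, used as both spawn group and skeletal mesh bone/socket index.
# Index-driven switches set sprite offset, alignment, facing, size and UV mirroring, with spawn-group masks extending the last slot.
# SubUV animation varies the line art, and camera offset differs between gameplay and front end.
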NiagaraSystem /BRCosmetics/Animation/Game/MainPlayer/Emotes/SulfurDean/VFX/NS_SulfurDean_ShakingLines
dependencies {
 usedDataInterfaces [
  &NiagaraDataInterfaceSpriteRendererInfo
  &NiagaraDataInterfaceCurve
  &NiagaraDataInterfaceSkeletalMesh
 ]
 usedDynamicInputs []
 usedModules [
  &SystemState
  &self:SystemSpawnScript.SystemScriptSource.SystemScriptGraph.NiagaraNodeAssignment_0.SetVariables_#1
  &BlendSpaceHelper
 ]
}
emitters {Speed_Lines {
  emitterUpdateScript [
   &EmitterState()
   &SpawnBurst_Instantaneous(Spawn Count=11, Use Spawn Probability=false, UseLoopCountLimit=false)
  ]
  particleSpawnScript [
   &PartitionParticles(Number Of Partitions=11)
   /SetVariables(Particles.RandomBones:NiagaraInt32=$Particles.PartitionParticles.Partition) {isSetParameters=true}
   /SetVariables(Particles.SpawnGroup=$Particles.RandomBones) {isSetParameters=true}
   &InitializeParticle(
    Color=&RandomRangeLinearColor(Minimum=<rgba 0 0 0 1>, Maximum=<rgba 0.01 0.01 0.01 1>)
    Lifetime=0.075
    Position Mode=::NewEnumerator0
    Sprite Rotation Mode=::NewEnumerator0
    Sprite Size=&MergeHelper_Vector2Switch(
     0=&RandomRangeVector2D(Minimum=<12 12>, Maximum=<13 13>)
     1=&RandomRangeVector2D(Minimum=<12 12>, Maximum=<13 13>)
     2=&RandomRangeVector2D(Minimum=<12 12>, Maximum=<13 13>)
     3=&RandomRangeVector2D(Minimum=<12 12>, Maximum=<13 13>)
     4=&RandomRangeVector2D(Minimum=<12 12>, Maximum=<13 13>)
     5=&RandomRangeVector2D(Minimum=<12 12>, Maximum=<13 13>)
     6=&RandomRangeVector2D(Minimum=<12 12>, Maximum=<13 13>)
     7=&RandomRangeVector2D(Minimum=<12 12>, Maximum=<13 13>)
     8=&RandomRangeVector2D(Minimum=<12 12>, Maximum=<13 13>)
     9=&RandomRangeVector2D(Minimum=<12 12>, Maximum=<13 13>)
     Index=&SpawnGroup()
     Num=::NewEnumerator8
    )
    Sprite Size Mode=::NewEnumerator4
    Sprite UV Mode=::NewEnumerator4
    Sprite UV Scale:Vector2f=&MergeHelper_Vector2Switch(
     0=<-1 -1>
     1=<-1 1>
     2=<1 -1>
     3=<1 1>
     4=<-1 -1>
     5=<-1 1>
     6=<1 1>
     7=<-1 -1>
     8=<-1 -1>
     9=<1 1>
     Index=&SpawnGroup()
     Num=::NewEnumerator8
    )
    Write Material Random=true
   )
   &SubUVAnimation(
    End Frame Range Override=8
    "Sprite Renderer "=DataInterface() {propertyValues {SpriteRenderer=&self:Speed_Lines_0.NiagaraSpriteRendererProperties_0}}
    SubUV Animation Mode:ENiagara_SubUVLookupModeV2=::NewEnumerator2
    UseEndFrame=true
    UseStartFrame=true
   )
   &CameraOffset(Camera Offset Amount=&FN_FrontEndFloatModifier(Gameplay=5, Front End=-30))
  ]
  particleUpdateScript [
   &ParticleState()
   &SkeletalMeshLocation(
    Bone / Socket Index=$Particles.RandomBones
    Bone Sampling Mode=::NewEnumerator4
    MaskSpawnGroup=false
    Sampled Position Offset=&MergeHelper_VectorSwitch(
     0=&RandomRangeVector(Minimum=<-0.891 -2 6>, Maximum=<-0.7953 -1 7>)
     1=&RandomRangeVector(Minimum=<0.891 2 -6>, Maximum=<0.891 1 -7>)
     2=<-6 -2 1>
     3=<-6 -3 4>
     4=<6 3 -4>
     5=<6 2 -1>
     6=<30 -5 4>
     7=<-30 5 -4>
     8=<35 -3 0>
     9=&MaskVectorBySpawnGroup(
      Masked Vector=<-35 3 0>
      Passthrough Vector=&MaskVectorBySpawnGroup(Masked Vector=<-10 -11 15>, Passthrough Vector=<-10 -11 -15>, Spawn Group Mask=10)
      Spawn Group Mask=9
     )
     Index=&MakeCustomIntFromBool(
      Bool=&SetBoolByIntComparison(A=$Particles.RandomBones, B=9, Comparison Type=::NewEnumerator0)
      True INT=9
      False INT=$Particles.RandomBones
     )
     Num=::NewEnumerator8
    )
    Skeletal Mesh=$System.Skeletal Mesh
   )
   &SpriteFacingAndAlignment(
    Sprite Alignment=&RotateVectorByQuaternion(
     Vector To Rotate=&MergeHelper_VectorSwitch(
      0=<1 0 0>
      1=<1 0 0>
      2=<0 1 0>
      3=<0 -1 0>
      4=<0 -1 0>
      5=<0 1 0>
      6=<1 -0.548 0>
      7=<1 -0.548 0>
      8=<0.19 0.645 0>
      9=&MaskVectorBySpawnGroup(Masked Vector=<0.19 0.645 0>, Passthrough Vector=<-1 0 0>, Spawn Group Mask=9)
      Index=&MakeCustomIntFromBool(
       Bool=&SetBoolByIntComparison(A=$Particles.RandomBones, B=9, Comparison Type=::NewEnumerator0)
       True INT=9
       False INT=$Particles.RandomBones
      )
      Num=::NewEnumerator8
     )
     Quaternion=$Particles.SkeletalMeshLocation.SampledRotation
    )
    Sprite Facing=&MergeHelper_VectorSwitch(
     0=<-1 0 0>
     1=<-1 0 0>
     2=<0 1 0>
     3=<1 0 0>
     4=<-1 0 0>
     5=<0 1 0>
     6=<1 0 0>
     7=<-1 0 0>
     8=<0 1 0>
     9=<0 1 0>
     Index=&MakeCustomIntFromBool(
      Bool=&SetBoolByIntComparison(A=$Particles.RandomBones, B=9, Comparison Type=::NewEnumerator0)
      True INT=9
      False INT=$Particles.RandomBones
     )
     Num=::NewEnumerator8
    )
    WriteAlignment=true
    WriteFacing=true
   )
   &Drag(Drag=5, Use Linear Drag=true, Use Rotational Drag=false)
   &SolveForcesAndVelocity(Clamp Velocity=false, Limit Acceleration=false)
   &ScaleColor(
    Scale Alpha=&FloatFromCurve(FloatCurve=DataInterface() {propertyValues {Curve {keys [
         {interpMode=RCIM_Cubic, tangentMode=RCTM_User, time=0.1552, value=1}
         {interpMode=RCIM_Cubic, tangentMode=RCTM_User, time=0.5, value=1}
         {tangentMode=RCTM_User, time=1, value=1}
        ]}}})
    ScaleA=true
    ScaleRGB=true
   )
   &JitterPosition(Jitter Amount=0.3596)
  ]
  rendererClasses [&NiagaraSpriteRendererProperties]
  renderers [{rendererClass=&NiagaraSpriteRendererProperties}]
 }}
summary {
 systemName=NS_SulfurDean_ShakingLines
 emitters [{emitterName=Speed_Lines, rendererClasses [&NiagaraSpriteRendererProperties]}]
 userVariables [
  User.bIsFrontEnd {type=NiagaraBool}
  User.bIsFrontEndPreview {type=NiagaraBool}
  User.Intensity {type=NiagaraFloat, defaultValue=0.68, description=""}
  User.OwnerLocalVelocity {type=Vector3f, description="", defaultValue {struct=Vector3f, value=<0 0 0>}}
  User.RibbonIDOffset {type=NiagaraInt32}
  User.WorkoutSelect {type=NiagaraFloat, defaultValue=0, description=""}
 ]
}
